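# Looping interior ambience Sound Cue for the train tunnel POI, routed to the indoor ambient sound class without spatialization, attenuation or reverb send.
# A quality-level node picks one of two mixes, each layering the underground lair cave loop over a half-volume stone room tone.
# The first two levels use the quad room tone, while the last two use the stereo version pitched up.
# Concurrency override prevents new instances at the limit.
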SoundCue /Game/Sounds/Helios/POI/TrainTunnels/Ambience_TrainTunnel_Interior_Loop_Cue
root=QualityLevel_1
nodes [
 {id=Mixer_1, children [WavePlayer_0,Modulator_0], properties {inputVolume [1,0.5]}}
 {id=Mixer_3, children [Modulator_3,WavePlayer_1], properties {inputVolume [0.5,1]}}
 {id=Modulator_0, children [WavePlayer_8], properties {pitchMax=1, pitchMin=1}}
 {id=Modulator_3, children [WavePlayer_9], properties {pitchMax=1.2, pitchMin=1.2}}
 {id=QualityLevel_1, children [Mixer_1,Mixer_1,Mixer_3,Mixer_3]}
 {
  id=WavePlayer_0
  properties {bLooping=true, soundWaveAssetPtr=/Game/Sounds/Helios/POI/UndergroundLair/Ambience_Underground_Lair_Cave_Main_Airy_Quad_Loop}
 }
 {
  id=WavePlayer_1
  properties {bLooping=true, soundWaveAssetPtr=&Ambience_Underground_Lair_Cave_Main_Airy_Quad_Loop}
 }
 {
  id=WavePlayer_8
  properties {bLooping=true, soundWaveAssetPtr=/Game/Sounds/Artemis/RoomTones/Stone/Artemis_Interior_Large_Stone_Room_02_Quad}
 }
 {
  id=WavePlayer_9
  properties {bLooping=true, soundWaveAssetPtr=/Game/Sounds/Artemis/RoomTones/Stone/Artemis_Interior_Large_Stone_Room_02_Stereo}
 }
]
properties {
 attenuationSettings=&Fort_Ambient_Athena_Att
 bOverrideConcurrency=true
 maxDistance=2097152
 soundClassObject=&Fort_Ambient_Indoors_Athena
 volumeMultiplier=0.501
 attenuationOverrides {bAttenuate=false, bEnableReverbSend=false, bSpatialize=false}
 concurrencyOverrides {resolutionRule=PreventNew}
}
